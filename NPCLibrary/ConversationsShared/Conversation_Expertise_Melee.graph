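# NPC conversation database for the melee Expertise entry, selling the ElectroBat through "Expert Sale" choices gated on the Kneecapper service.
# Three parallel speech branches each offer Back or an ExpertSale purchase, filtered by player expertise-token thresholds and priced from ExpertSalePricing curve rows One, Two and Three.
# Purchases grant the same loot tier group, remove Wads, play effects, then apply a data-driven effect.
# The tier Three sale checks row Three but deducts row One.
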ConversationDatabase /NPCLibrary/ConversationsShared/Conversation_Expertise_Melee
entry_points [{tag=Conversation.Entry.Expertise.Melee, destinations [#1]}]
nodes [
 ConversationEntryPointNode_3 {
  id=#1
  class=ConversationEntryPointNode
  outputs [#2, #3, #4]
  properties {entryTag {tagName=Conversation.Entry.Expertise.Melee}}
 }
 Speech_C_1 {
  id=#4
  class=Speech_C
  outputs [#5, #6]
  properties {bUseInlineConfig=false, softTaskIcon=&T_UI_NPC_WeaponExpertise}
  sub_nodes [
   FortConversationChoiceWithPreviewNode_1 {class=FortConversationChoiceWithPreviewNode, properties {defaultChoiceDisplayText=EXPERT SALE}}
   FortConversationRequirement_HasService_2 {class=FortConversationRequirement_HasService, properties {serviceTag {tagName=Tandem.Service.ExpertSale.Kneecapper}}}
  ]
 }
 ExpertSale_C_20 {
  id=#6
  class=ExpertSale_C
  outputs [#7]
  properties {
   bShowPropertyDetails=false
   requirements [
    {
     cannotUseReasonParameter=notExpertEnough
     failureNodeBehaviour=FailedAndHidden
     participantId {tagName=Conversation.Participant.Player}
     requirement {class=UFortControllerRequirement_TokenThreshold_C, properties {requiredQuantityThresholdConfig {quantity {value=3}}}}
    }
    {
     cannotUseReasonParameter=notEnoughResources
     failureNodeBehaviour=FailedButVisible
     participantId {tagName=Conversation.Participant.Player}
     requirement {class=FortControllerRequirement_HasResource, properties {thresholdTestConfig {quantity {curve {rowName=ExpertSale.Cost.ElectroBat.Three}}}}}
    }
    {
     cannotUseReasonParameter=notExpertEnough
     failureNodeBehaviour=FailedButVisible
     participantId {tagName=Conversation.Participant.Player}
     requirement {
      class=UFortControllerRequirement_TokenThreshold_C
      properties {requiredQuantityThresholdConfig {requirement=GreaterThanOrEqual, quantity {value=3}}}
     }
    }
   ]
   serviceCost {curve {rowName=ExpertSale.Cost.ElectroBat.Three}}
  }
  sub_nodes [
   ConversationChoiceNode_22 {class=ConversationChoiceNode, properties {defaultChoiceDisplayText=ACQUIRE}}
   SideEffect_PlayServiceUsedEffects_C_20 {class=SideEffect_PlayServiceUsedEffects_C}
   ConversationSideEffect_GiveItemFromLootTierGroup_C_20 {
    class=ConversationSideEffect_GiveItemFromLootTierGroup_C
    properties {bShowPropertyDetails=false, lootTierGroupName=LTG_NPC_Expertise_ElectroBat_VR}
   }
   ConversationSideEffect_RemoveResource_C_20 {
    class=ConversationSideEffect_RemoveResource_C
    properties {
     bShowPropertyDetails=false
     item Definition=/Game/Items/ResourcePickups/Athena_WadsItemData
     amountToRemove {value=1, curve {curveTable=/Game/Conversation/Services/ExpertSale/DataTables/ExpertSalePricing, rowName=ExpertSale.Cost.ElectroBat.One}}
    }
   }
  ]
 }
 FortConversationTaskNode_ApplyDataDrivenEffect_2 {id=#7, class=FortConversationTaskNode_ApplyDataDrivenEffect, outputs []}
 Back_C_3 {
  id=#5
  class=Back_C
  sub_nodes [ConversationChoiceNode_40 {class=ConversationChoiceNode, properties {defaultChoiceDisplayText=BACK}}]
 }
 Speech_C_4 {
  id=#3
  class=Speech_C
  outputs [#8, #9]
  properties {bUseInlineConfig=false, softTaskIcon=&T_UI_NPC_WeaponExpertise}
  sub_nodes [
   FortConversationChoiceWithPreviewNode_12 {class=FortConversationChoiceWithPreviewNode, properties {defaultChoiceDisplayText=EXPERT SALE}}
   FortConversationRequirement_HasService_11 {class=FortConversationRequirement_HasService, properties {serviceTag {tagName=Tandem.Service.ExpertSale.Kneecapper}}}
  ]
 }
 ExpertSale_C_18 {
  id=#9
  class=ExpertSale_C
  outputs [#7]
  properties {
   bShowPropertyDetails=false
   extraData [
    item {value=/ElectroBatGameplay/Gameplay/WID_Athena_ElectroBat}
    cost {value=600}
    remaining {value=1}
    stacksize {value=1}
    Type {value=service.expertsale}
    preview_tags {value=(GameplayTags=((TagName="Conversation.Choice.Category.Service")))}
    BriefWidgetClassPath {value=/FortniteConversationUI/Conversation/WBP_FortBuyBrief_ExpertSale.WBP_FortBuyBrief_ExpertSale_C}
    stockcount {value=1}
    configdef {value=/Game/Conversation/Services/ExpertSale/ExpertSaleBriefCollection}
    configkey {value=ExpertSale}
    cannotusereason {value=notExpertEnough}
    stockcount {value=0}
    cannotusereason {value=outofstock}
   ]
   requirements [
    {
     cannotUseReasonParameter=notExpertEnough
     failureNodeBehaviour=FailedAndHidden
     participantId {tagName=Conversation.Participant.Player}
     requirement {class=UFortControllerRequirement_TokenThreshold_C, properties {requiredQuantityThresholdConfig {quantity {value=2}}}}
    }
    {
     cannotUseReasonParameter=notEnoughResources
     failureNodeBehaviour=FailedButVisible
     participantId {tagName=Conversation.Participant.Player}
     requirement {class=FortControllerRequirement_HasResource, properties {thresholdTestConfig {quantity {curve {rowName=ExpertSale.Cost.ElectroBat.One}}}}}
    }
    {
     cannotUseReasonParameter=notExpertEnough
     failureNodeBehaviour=FailedButVisible
     participantId {tagName=Conversation.Participant.Player}
     requirement {class=UFortControllerRequirement_TokenThreshold_C, properties {requiredQuantityThresholdConfig {requirement=Equal, quantity=~}}}
    }
   ]
   serviceCost {curve {rowName=ExpertSale.Cost.ElectroBat.One}}
  }
  sub_nodes [
   ConversationChoiceNode_23 {class=ConversationChoiceNode, properties {defaultChoiceDisplayText=ACQUIRE}}
   SideEffect_PlayServiceUsedEffects_C_21 {class=SideEffect_PlayServiceUsedEffects_C}
   ConversationSideEffect_GiveItemFromLootTierGroup_C_21 {
    class=ConversationSideEffect_GiveItemFromLootTierGroup_C
    properties {bShowPropertyDetails=false, lootTierGroupName=LTG_NPC_Expertise_ElectroBat_VR}
   }
   ConversationSideEffect_RemoveResource_C_21 {
    class=ConversationSideEffect_RemoveResource_C
    properties {
     bShowPropertyDetails=false
     item Definition=&Athena_WadsItemData
     amountToRemove {value=1, curve {curveTable=&ExpertSalePricing, rowName=ExpertSale.Cost.ElectroBat.One}}
    }
   }
  ]
 }
 Back_C_2 {
  id=#8
  class=Back_C
  sub_nodes [ConversationChoiceNode_12 {class=ConversationChoiceNode, properties {defaultChoiceDisplayText=BACK}}]
 }
 Speech_C_3 {
  id=#2
  class=Speech_C
  outputs [#10, #11]
  properties {bUseInlineConfig=false, softTaskIcon=&T_UI_NPC_WeaponExpertise}
  sub_nodes [
   FortConversationChoiceWithPreviewNode_11 {class=FortConversationChoiceWithPreviewNode, properties {defaultChoiceDisplayText=EXPERT SALE}}
   FortConversationRequirement_HasService_10 {class=FortConversationRequirement_HasService, properties {serviceTag {tagName=Tandem.Service.ExpertSale.Kneecapper}}}
  ]
 }
 ExpertSale_C_19 {
  id=#11
  class=ExpertSale_C
  outputs [#7]
  properties {
   bShowPropertyDetails=false
   extraData [
    item {value=&WID_Athena_ElectroBat}
    cost {value=240}
    remaining {value=1}
    stacksize {value=1}
    Type {value=service.expertsale}
    preview_tags {value=(GameplayTags=((TagName="Conversation.Choice.Category.Service")))}
    BriefWidgetClassPath {value=&WBP_FortBuyBrief_ExpertSale_C}
    stockcount {value=1}
    configdef {value=&ExpertSaleBriefCollection}
    configkey {value=ExpertSale}
    cost {value=420}
   ]
   requirements [
    {
     cannotUseReasonParameter=notExpertEnough
     failureNodeBehaviour=FailedAndHidden
     participantId {tagName=Conversation.Participant.Player}
     requirement {class=UFortControllerRequirement_TokenThreshold_C, properties {requiredQuantityThresholdConfig {quantity {value=2}}}}
    }
    {
     cannotUseReasonParameter=notEnoughResources
     failureNodeBehaviour=FailedButVisible
     participantId {tagName=Conversation.Participant.Player}
     requirement {class=FortControllerRequirement_HasResource, properties {thresholdTestConfig {quantity {curve {rowName=ExpertSale.Cost.ElectroBat.Two}}}}}
    }
    {
     cannotUseReasonParameter=notExpertEnough
     failureNodeBehaviour=FailedButVisible
     participantId {tagName=Conversation.Participant.Player}
     requirement {class=UFortControllerRequirement_TokenThreshold_C, properties {requiredQuantityThresholdConfig {requirement=Equal, quantity {value=2}}}}
    }
   ]
   serviceCost {curve {rowName=ExpertSale.Cost.ElectroBat.Two}}
  }
  sub_nodes [
   ConversationChoiceNode_24 {class=ConversationChoiceNode, properties {defaultChoiceDisplayText=ACQUIRE}}
   SideEffect_PlayServiceUsedEffects_C_22 {class=SideEffect_PlayServiceUsedEffects_C}
   ConversationSideEffect_GiveItemFromLootTierGroup_C_22 {
    class=ConversationSideEffect_GiveItemFromLootTierGroup_C
    properties {bShowPropertyDetails=false, lootTierGroupName=LTG_NPC_Expertise_ElectroBat_VR}
   }
   ConversationSideEffect_RemoveResource_C_22 {
    class=ConversationSideEffect_RemoveResource_C
    properties {
     bShowPropertyDetails=false
     item Definition=&Athena_WadsItemData
     amountToRemove {value=1, curve {curveTable=&ExpertSalePricing, rowName=ExpertSale.Cost.ElectroBat.Two}}
    }
   }
  ]
 }
 Back_C_19 {
  id=#10
  class=Back_C
  sub_nodes [ConversationChoiceNode_10 {class=ConversationChoiceNode, properties {defaultChoiceDisplayText=BACK}}]
 }
]
