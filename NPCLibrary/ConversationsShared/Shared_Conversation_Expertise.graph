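# Shared NPC conversation database for the Expertise entry point, offering Expert Sale purchases to players.
# Six Speech branches, gated by the NPC's Healing (one) or Exotic (five) service tag, each lead to Back or BUY, which returns to start.
# Exotic sales share one loot group, differing in cost row and exact expertise-token count (one to five), and stay visible when failed.
# The Healing sale's token check is marked skipped.
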ConversationDatabase /NPCLibrary/ConversationsShared/Shared_Conversation_Expertise
entry_points [{tag=Conversation.Entry.Expertise, destinations [#1]}]
nodes [
 ConversationEntryPointNode_0 {
  id=#1
  class=ConversationEntryPointNode
  outputs [
   #2
   #3
   #4
   #5
   #6
   #7
  ]
  properties {entryTag {tagName=Conversation.Entry.Expertise}}
 }
 Speech_C_7 {
  id=#7
  class=Speech_C
  outputs [#8, #9]
  properties {softTaskIcon=&T_UI_NPC_WeaponExpertise, generalConfig {defaultMessage=For experts only.}}
  sub_nodes [
   FortConversationChoiceWithPreviewNode_4 {class=FortConversationChoiceWithPreviewNode, properties {defaultChoiceDisplayText=EXPERT SALE}}
   FortConversationRequirement_HasService_4 {class=FortConversationRequirement_HasService, properties {serviceTag {tagName=Tandem.Service.ExpertSale.Exotic}}}
  ]
 }
 ExpertSale_C_14 {
  id=#9
  class=ExpertSale_C
  outputs [#10]
  properties {
   bShowPropertyDetails=false
   loot Tier Group Name=LTG_NPC_ExoticExpertise_DiamondItemData
   extraData [
    locked {value=1}
    item {value=/DragonPopGameplayContent/Gameplay/DiamondCurrency/Athena_DiamondItemData}
    cost {value=900}
    remaining {value=1}
    stacksize {value=1}
    locked_text {value="Expertise Required!"}
    Type {value=service.expertsale}
    preview_tags {value=(GameplayTags=((TagName="Conversation.Choice.Category.Service")))}
    BriefWidgetClassPath {value=/FortniteConversationUI/Conversation/WBP_FortBuyBrief_ExpertSale.WBP_FortBuyBrief_ExpertSale_C}
    stockcount {value=1}
    cannotusereason {value=""}
    configdef {value=/Game/Conversation/Services/ExpertSale/ExpertSaleBriefCollection}
    configkey {value=ExpertSale}
    Type {value=service.expertsale}
    preview_tags {value=(GameplayTags=((TagName="Conversation.Choice.Category.Service")))}
    BriefWidgetClassPath {value=&WBP_FortBuyBrief_ExpertSale_C}
    stockcount {value=1}
    cannotusereason {value=""}
    configdef {value=&ExpertSaleBriefCollection}
    configkey {value=ExpertSale}
    Type {value=service.expertsale}
    preview_tags {value=(GameplayTags=((TagName="Conversation.Choice.Category.Service")))}
    BriefWidgetClassPath {value=&WBP_FortBuyBrief_ExpertSale_C}
    stockcount {value=1}
    cannotusereason {value=""}
    configdef {value=&ExpertSaleBriefCollection}
    configkey {value=ExpertSale}
    Type {value=service.expertsale}
    preview_tags {value=(GameplayTags=((TagName="Conversation.Choice.Category.Service")))}
    BriefWidgetClassPath {value=&WBP_FortBuyBrief_ExpertSale_C}
    stockcount {value=1}
    cannotusereason {value=""}
    configdef {value=&ExpertSaleBriefCollection}
    configkey {value=ExpertSale}
    Type {value=service.expertsale}
    preview_tags {value=(GameplayTags=((TagName="Conversation.Choice.Category.Service")))}
    BriefWidgetClassPath {value=&WBP_FortBuyBrief_ExpertSale_C}
    stockcount {value=1}
    cannotusereason {value=""}
    configdef {value=&ExpertSaleBriefCollection}
    configkey {value=ExpertSale}
    Type {value=service.expertsale}
    preview_tags {value=(GameplayTags=((TagName="Conversation.Choice.Category.Service")))}
    BriefWidgetClassPath {value=&WBP_FortBuyBrief_ExpertSale_C}
    stockcount {value=1}
    cannotusereason {value=""}
    configdef {value=&ExpertSaleBriefCollection}
    configkey {value=ExpertSale}
    Type {value=service.expertsale}
    preview_tags {value=(GameplayTags=((TagName="Conversation.Choice.Category.Service")))}
    BriefWidgetClassPath {value=&WBP_FortBuyBrief_ExpertSale_C}
    stockcount {value=1}
    cannotusereason {value=""}
    configdef {value=&ExpertSaleBriefCollection}
    configkey {value=ExpertSale}
    Type {value=service.expertsale}
    preview_tags {value=(GameplayTags=((TagName="Conversation.Choice.Category.Service")))}
    BriefWidgetClassPath {value=&WBP_FortBuyBrief_ExpertSale_C}
    stockcount {value=1}
    cannotusereason {value=""}
    configdef {value=&ExpertSaleBriefCollection}
    configkey {value=ExpertSale}
   ]
   requirements [{
     cannotUseReasonParameter=""
     failureNodeBehaviour=FailedButVisible
     participantId {tagName=Conversation.Participant.Player}
     requirement {
      class=UFortControllerRequirement_TokenThreshold_C
      properties {
       itemTemplateId=Token:athena_s34_weaponexpertise_exotic_service_token
       requiredQuantityThresholdConfig {
        requirement=Equal
        quantity {
         value=5
         curve {curveTable=/Game/Conversation/Services/ExpertSale/DataTables/ExpertSaleRequirementsData, rowName=ExpertSale.Requirement.Tokens}
        }
       }
      }
     }
    }]
   serviceCost {curve {rowName=ExpertSale.Cost.Exotic.Five}}
  }
  sub_nodes [ConversationChoiceNode_14 {class=ConversationChoiceNode, properties {defaultChoiceDisplayText=BUY}}]
 }
 ReturnToStart_C_0 {id=#10, class=ReturnToStart_C}
 Back_C_5 {
  id=#8
  class=Back_C
  sub_nodes [ConversationChoiceNode_0 {class=ConversationChoiceNode, properties {defaultChoiceDisplayText=BACK}}]
 }
 Speech_C_6 {
  id=#6
  class=Speech_C
  outputs [#11, #12]
  properties {softTaskIcon=&T_UI_NPC_WeaponExpertise, generalConfig {defaultMessage=For experts only.}}
  sub_nodes [
   FortConversationChoiceWithPreviewNode_7 {class=FortConversationChoiceWithPreviewNode, properties {defaultChoiceDisplayText=EXPERT SALE}}
   FortConversationRequirement_HasService_8 {class=FortConversationRequirement_HasService, properties {serviceTag {tagName=Tandem.Service.ExpertSale.Exotic}}}
  ]
 }
 ExpertSale_C_13 {
  id=#12
  class=ExpertSale_C
  outputs [#10]
  properties {
   bShowPropertyDetails=false
   loot Tier Group Name=LTG_NPC_ExoticExpertise_DiamondItemData
   extraData [
    locked {value=1}
    item {value=&Athena_DiamondItemData}
    cost {value=900}
    remaining {value=1}
    stacksize {value=1}
    locked_text {value="Expertise Required!"}
    Type {value=service.expertsale}
    preview_tags {value=(GameplayTags=((TagName="Conversation.Choice.Category.Service")))}
    BriefWidgetClassPath {value=&WBP_FortBuyBrief_ExpertSale_C}
    stockcount {value=1}
    cannotusereason {value=""}
    configdef {value=&ExpertSaleBriefCollection}
    configkey {value=ExpertSale}
    Type {value=service.expertsale}
    preview_tags {value=(GameplayTags=((TagName="Conversation.Choice.Category.Service")))}
    BriefWidgetClassPath {value=&WBP_FortBuyBrief_ExpertSale_C}
    stockcount {value=1}
    cannotusereason {value=""}
    configdef {value=&ExpertSaleBriefCollection}
    configkey {value=ExpertSale}
    Type {value=service.expertsale}
    preview_tags {value=(GameplayTags=((TagName="Conversation.Choice.Category.Service")))}
    BriefWidgetClassPath {value=&WBP_FortBuyBrief_ExpertSale_C}
    stockcount {value=1}
    cannotusereason {value=""}
    configdef {value=&ExpertSaleBriefCollection}
    configkey {value=ExpertSale}
    Type {value=service.expertsale}
    preview_tags {value=(GameplayTags=((TagName="Conversation.Choice.Category.Service")))}
    BriefWidgetClassPath {value=&WBP_FortBuyBrief_ExpertSale_C}
    stockcount {value=1}
    cannotusereason {value=""}
    configdef {value=&ExpertSaleBriefCollection}
    configkey {value=ExpertSale}
    Type {value=service.expertsale}
    preview_tags {value=(GameplayTags=((TagName="Conversation.Choice.Category.Service")))}
    BriefWidgetClassPath {value=&WBP_FortBuyBrief_ExpertSale_C}
    stockcount {value=1}
    cannotusereason {value=""}
    configdef {value=&ExpertSaleBriefCollection}
    configkey {value=ExpertSale}
    Type {value=service.expertsale}
    preview_tags {value=(GameplayTags=((TagName="Conversation.Choice.Category.Service")))}
    BriefWidgetClassPath {value=&WBP_FortBuyBrief_ExpertSale_C}
    stockcount {value=1}
    cannotusereason {value=""}
    configdef {value=&ExpertSaleBriefCollection}
    configkey {value=ExpertSale}
    Type {value=service.expertsale}
    preview_tags {value=(GameplayTags=((TagName="Conversation.Choice.Category.Service")))}
    BriefWidgetClassPath {value=&WBP_FortBuyBrief_ExpertSale_C}
    stockcount {value=1}
    cannotusereason {value=""}
    configdef {value=&ExpertSaleBriefCollection}
    configkey {value=ExpertSale}
    Type {value=service.expertsale}
    preview_tags {value=(GameplayTags=((TagName="Conversation.Choice.Category.Service")))}
    BriefWidgetClassPath {value=&WBP_FortBuyBrief_ExpertSale_C}
    stockcount {value=1}
    cannotusereason {value=""}
    configdef {value=&ExpertSaleBriefCollection}
    configkey {value=ExpertSale}
   ]
   requirements [{
     cannotUseReasonParameter=""
     failureNodeBehaviour=FailedButVisible
     participantId {tagName=Conversation.Participant.Player}
     requirement {
      class=UFortControllerRequirement_TokenThreshold_C
      properties {
       itemTemplateId=Token:athena_s34_weaponexpertise_exotic_service_token
       requiredQuantityThresholdConfig {
        requirement=Equal
        quantity {
         value=4
         curve {curveTable=&ExpertSaleRequirementsData, rowName=ExpertSale.Requirement.Tokens}
        }
       }
      }
     }
    }]
   serviceCost {curve {rowName=ExpertSale.Cost.Exotic.Four}}
  }
  sub_nodes [ConversationChoiceNode_20 {class=ConversationChoiceNode, properties {defaultChoiceDisplayText=BUY}}]
 }
 Back_C_4 {
  id=#11
  class=Back_C
  sub_nodes [ConversationChoiceNode_19 {class=ConversationChoiceNode, properties {defaultChoiceDisplayText=BACK}}]
 }
 Speech_C_4 {
  id=#5
  class=Speech_C
  outputs [#13, #14]
  properties {softTaskIcon=&T_UI_NPC_WeaponExpertise, generalConfig {defaultMessage=For experts only.}}
  sub_nodes [
   FortConversationChoiceWithPreviewNode_6 {class=FortConversationChoiceWithPreviewNode, properties {defaultChoiceDisplayText=EXPERT SALE}}
   FortConversationRequirement_HasService_7 {class=FortConversationRequirement_HasService, properties {serviceTag {tagName=Tandem.Service.ExpertSale.Exotic}}}
  ]
 }
 ExpertSale_C_12 {
  id=#14
  class=ExpertSale_C
  outputs [#10]
  properties {
   bShowPropertyDetails=false
   loot Tier Group Name=LTG_NPC_ExoticExpertise_DiamondItemData
   extraData [
    locked {value=1}
    item {value=&Athena_DiamondItemData}
    cost {value=900}
    remaining {value=1}
    stacksize {value=1}
    locked_text {value="Expertise Required!"}
    Type {value=service.expertsale}
    preview_tags {value=(GameplayTags=((TagName="Conversation.Choice.Category.Service")))}
    BriefWidgetClassPath {value=&WBP_FortBuyBrief_ExpertSale_C}
    stockcount {value=1}
    cannotusereason {value=""}
    configdef {value=&ExpertSaleBriefCollection}
    configkey {value=ExpertSale}
    Type {value=service.expertsale}
    preview_tags {value=(GameplayTags=((TagName="Conversation.Choice.Category.Service")))}
    BriefWidgetClassPath {value=&WBP_FortBuyBrief_ExpertSale_C}
    stockcount {value=1}
    cannotusereason {value=""}
    configdef {value=&ExpertSaleBriefCollection}
    configkey {value=ExpertSale}
    Type {value=service.expertsale}
    preview_tags {value=(GameplayTags=((TagName="Conversation.Choice.Category.Service")))}
    BriefWidgetClassPath {value=&WBP_FortBuyBrief_ExpertSale_C}
    stockcount {value=1}
    cannotusereason {value=""}
    configdef {value=&ExpertSaleBriefCollection}
    configkey {value=ExpertSale}
    Type {value=service.expertsale}
    preview_tags {value=(GameplayTags=((TagName="Conversation.Choice.Category.Service")))}
    BriefWidgetClassPath {value=&WBP_FortBuyBrief_ExpertSale_C}
    stockcount {value=1}
    cannotusereason {value=""}
    configdef {value=&ExpertSaleBriefCollection}
    configkey {value=ExpertSale}
    Type {value=service.expertsale}
    preview_tags {value=(GameplayTags=((TagName="Conversation.Choice.Category.Service")))}
    BriefWidgetClassPath {value=&WBP_FortBuyBrief_ExpertSale_C}
    stockcount {value=1}
    cannotusereason {value=""}
    configdef {value=&ExpertSaleBriefCollection}
    configkey {value=ExpertSale}
    Type {value=service.expertsale}
    preview_tags {value=(GameplayTags=((TagName="Conversation.Choice.Category.Service")))}
    BriefWidgetClassPath {value=&WBP_FortBuyBrief_ExpertSale_C}
    stockcount {value=1}
    cannotusereason {value=""}
    configdef {value=&ExpertSaleBriefCollection}
    configkey {value=ExpertSale}
    Type {value=service.expertsale}
    preview_tags {value=(GameplayTags=((TagName="Conversation.Choice.Category.Service")))}
    BriefWidgetClassPath {value=&WBP_FortBuyBrief_ExpertSale_C}
    stockcount {value=1}
    cannotusereason {value=""}
    configdef {value=&ExpertSaleBriefCollection}
    configkey {value=ExpertSale}
    Type {value=service.expertsale}
    preview_tags {value=(GameplayTags=((TagName="Conversation.Choice.Category.Service")))}
    BriefWidgetClassPath {value=&WBP_FortBuyBrief_ExpertSale_C}
    stockcount {value=1}
    cannotusereason {value=""}
    configdef {value=&ExpertSaleBriefCollection}
    configkey {value=ExpertSale}
   ]
   requirements [{
     cannotUseReasonParameter=""
     failureNodeBehaviour=FailedButVisible
     participantId {tagName=Conversation.Participant.Player}
     requirement {
      class=UFortControllerRequirement_TokenThreshold_C
      properties {
       itemTemplateId=Token:athena_s34_weaponexpertise_exotic_service_token
       requiredQuantityThresholdConfig {
        requirement=Equal
        quantity {
         value=3
         curve {curveTable=&ExpertSaleRequirementsData, rowName=ExpertSale.Requirement.Tokens}
        }
       }
      }
     }
    }]
   serviceCost {curve {rowName=ExpertSale.Cost.Exotic.Three}}
  }
  sub_nodes [ConversationChoiceNode_18 {class=ConversationChoiceNode, properties {defaultChoiceDisplayText=BUY}}]
 }
 Back_C_3 {
  id=#13
  class=Back_C
  sub_nodes [ConversationChoiceNode_17 {class=ConversationChoiceNode, properties {defaultChoiceDisplayText=BACK}}]
 }
 Speech_C_5 {
  id=#4
  class=Speech_C
  outputs [#15, #16]
  properties {softTaskIcon=&T_UI_NPC_WeaponExpertise, generalConfig {defaultMessage=For experts only.}}
  sub_nodes [
   FortConversationChoiceWithPreviewNode_5 {class=FortConversationChoiceWithPreviewNode, properties {defaultChoiceDisplayText=EXPERT SALE}}
   FortConversationRequirement_HasService_6 {class=FortConversationRequirement_HasService, properties {serviceTag {tagName=Tandem.Service.ExpertSale.Exotic}}}
  ]
 }
 ExpertSale_C_11 {
  id=#16
  class=ExpertSale_C
  outputs [#10]
  properties {
   bShowPropertyDetails=false
   loot Tier Group Name=LTG_NPC_ExoticExpertise_DiamondItemData
   extraData [
    locked {value=1}
    item {value=&Athena_DiamondItemData}
    cost {value=900}
    remaining {value=1}
    stacksize {value=1}
    locked_text {value="Expertise Required!"}
    Type {value=service.expertsale}
    preview_tags {value=(GameplayTags=((TagName="Conversation.Choice.Category.Service")))}
    BriefWidgetClassPath {value=&WBP_FortBuyBrief_ExpertSale_C}
    stockcount {value=1}
    cannotusereason {value=""}
    configdef {value=&ExpertSaleBriefCollection}
    configkey {value=ExpertSale}
    Type {value=service.expertsale}
    preview_tags {value=(GameplayTags=((TagName="Conversation.Choice.Category.Service")))}
    BriefWidgetClassPath {value=&WBP_FortBuyBrief_ExpertSale_C}
    stockcount {value=1}
    cannotusereason {value=""}
    configdef {value=&ExpertSaleBriefCollection}
    configkey {value=ExpertSale}
    Type {value=service.expertsale}
    preview_tags {value=(GameplayTags=((TagName="Conversation.Choice.Category.Service")))}
    BriefWidgetClassPath {value=&WBP_FortBuyBrief_ExpertSale_C}
    stockcount {value=1}
    cannotusereason {value=""}
    configdef {value=&ExpertSaleBriefCollection}
    configkey {value=ExpertSale}
    Type {value=service.expertsale}
    preview_tags {value=(GameplayTags=((TagName="Conversation.Choice.Category.Service")))}
    BriefWidgetClassPath {value=&WBP_FortBuyBrief_ExpertSale_C}
    stockcount {value=1}
    cannotusereason {value=""}
    configdef {value=&ExpertSaleBriefCollection}
    configkey {value=ExpertSale}
    locked {value=0}
    Type {value=service.expertsale}
    preview_tags {value=(GameplayTags=((TagName="Conversation.Choice.Category.Service")))}
    BriefWidgetClassPath {value=&WBP_FortBuyBrief_ExpertSale_C}
    stockcount {value=1}
    configdef {value=&ExpertSaleBriefCollection}
    configkey {value=ExpertSale}
    Type {value=service.expertsale}
    preview_tags {value=(GameplayTags=((TagName="Conversation.Choice.Category.Service")))}
    BriefWidgetClassPath {value=&WBP_FortBuyBrief_ExpertSale_C}
    stockcount {value=1}
    configdef {value=&ExpertSaleBriefCollection}
    configkey {value=ExpertSale}
    Type {value=service.expertsale}
    preview_tags {value=(GameplayTags=((TagName="Conversation.Choice.Category.Service")))}
    BriefWidgetClassPath {value=&WBP_FortBuyBrief_ExpertSale_C}
    stockcount {value=1}
    configdef {value=&ExpertSaleBriefCollection}
    configkey {value=ExpertSale}
    Type {value=service.expertsale}
    preview_tags {value=(GameplayTags=((TagName="Conversation.Choice.Category.Service")))}
    BriefWidgetClassPath {value=&WBP_FortBuyBrief_ExpertSale_C}
    stockcount {value=1}
    configdef {value=&ExpertSaleBriefCollection}
    configkey {value=ExpertSale}
    Type {value=service.expertsale}
    preview_tags {value=(GameplayTags=((TagName="Conversation.Choice.Category.Service")))}
    BriefWidgetClassPath {value=&WBP_FortBuyBrief_ExpertSale_C}
    stockcount {value=1}
    configdef {value=&ExpertSaleBriefCollection}
    configkey {value=ExpertSale}
    Type {value=service.expertsale}
    preview_tags {value=(GameplayTags=((TagName="Conversation.Choice.Category.Service")))}
    BriefWidgetClassPath {value=&WBP_FortBuyBrief_ExpertSale_C}
    stockcount {value=1}
    configdef {value=&ExpertSaleBriefCollection}
    configkey {value=ExpertSale}
   ]
   requirements [{
     cannotUseReasonParameter=""
     failureNodeBehaviour=FailedButVisible
     participantId {tagName=Conversation.Participant.Player}
     requirement {
      class=UFortControllerRequirement_TokenThreshold_C
      properties {
       itemTemplateId=Token:athena_s34_weaponexpertise_exotic_service_token
       requiredQuantityThresholdConfig {
        requirement=Equal
        quantity {
         value=2
         curve {curveTable=&ExpertSaleRequirementsData, rowName=ExpertSale.Requirement.Tokens}
        }
       }
      }
     }
    }]
   serviceCost {curve {rowName=ExpertSale.Cost.Exotic.Two}}
  }
  sub_nodes [ConversationChoiceNode_16 {class=ConversationChoiceNode, properties {defaultChoiceDisplayText=BUY}}]
 }
 Back_C_2 {
  id=#15
  class=Back_C
  sub_nodes [ConversationChoiceNode_9 {class=ConversationChoiceNode, properties {defaultChoiceDisplayText=BACK}}]
 }
 Speech_C_2 {
  id=#3
  class=Speech_C
  outputs [#17, #18]
  properties {softTaskIcon=&T_UI_NPC_WeaponExpertise, generalConfig {defaultMessage=For experts only.}}
  sub_nodes [
   FortConversationChoiceWithPreviewNode_1 {class=FortConversationChoiceWithPreviewNode, properties {defaultChoiceDisplayText=EXPERT SALE}}
   FortConversationRequirement_HasService_2 {class=FortConversationRequirement_HasService, properties {serviceTag {tagName=Tandem.Service.ExpertSale.Exotic}}}
  ]
 }
 ExpertSale_C_10 {
  id=#18
  class=ExpertSale_C
  outputs [#10]
  properties {
   bShowPropertyDetails=false
   loot Tier Group Name=LTG_NPC_ExoticExpertise_DiamondItemData
   extraData [
    locked {value=1}
    item {value=&Athena_DiamondItemData}
    cost {value=900}
    remaining {value=1}
    stacksize {value=1}
    locked_text {value="Expertise Required!"}
    Type {value=service.expertsale}
    preview_tags {value=(GameplayTags=((TagName="Conversation.Choice.Category.Service")))}
    BriefWidgetClassPath {value=&WBP_FortBuyBrief_ExpertSale_C}
    stockcount {value=1}
    cannotusereason {value=""}
    configdef {value=&ExpertSaleBriefCollection}
    configkey {value=ExpertSale}
    Type {value=service.expertsale}
    preview_tags {value=(GameplayTags=((TagName="Conversation.Choice.Category.Service")))}
    BriefWidgetClassPath {value=&WBP_FortBuyBrief_ExpertSale_C}
    stockcount {value=1}
    cannotusereason {value=""}
    configdef {value=&ExpertSaleBriefCollection}
    configkey {value=ExpertSale}
    Type {value=service.expertsale}
    preview_tags {value=(GameplayTags=((TagName="Conversation.Choice.Category.Service")))}
    BriefWidgetClassPath {value=&WBP_FortBuyBrief_ExpertSale_C}
    stockcount {value=1}
    cannotusereason {value=""}
    configdef {value=&ExpertSaleBriefCollection}
    configkey {value=ExpertSale}
    Type {value=service.expertsale}
    preview_tags {value=(GameplayTags=((TagName="Conversation.Choice.Category.Service")))}
    BriefWidgetClassPath {value=&WBP_FortBuyBrief_ExpertSale_C}
    stockcount {value=1}
    cannotusereason {value=""}
    configdef {value=&ExpertSaleBriefCollection}
    configkey {value=ExpertSale}
    Type {value=service.expertsale}
    preview_tags {value=(GameplayTags=((TagName="Conversation.Choice.Category.Service")))}
    BriefWidgetClassPath {value=&WBP_FortBuyBrief_ExpertSale_C}
    stockcount {value=1}
    cannotusereason {value=""}
    configdef {value=&ExpertSaleBriefCollection}
    configkey {value=ExpertSale}
    Type {value=service.expertsale}
    preview_tags {value=(GameplayTags=((TagName="Conversation.Choice.Category.Service")))}
    BriefWidgetClassPath {value=&WBP_FortBuyBrief_ExpertSale_C}
    stockcount {value=1}
    cannotusereason {value=""}
    configdef {value=&ExpertSaleBriefCollection}
    configkey {value=ExpertSale}
    Type {value=service.expertsale}
    preview_tags {value=(GameplayTags=((TagName="Conversation.Choice.Category.Service")))}
    BriefWidgetClassPath {value=&WBP_FortBuyBrief_ExpertSale_C}
    stockcount {value=1}
    cannotusereason {value=""}
    configdef {value=&ExpertSaleBriefCollection}
    configkey {value=ExpertSale}
    Type {value=service.expertsale}
    preview_tags {value=(GameplayTags=((TagName="Conversation.Choice.Category.Service")))}
    BriefWidgetClassPath {value=&WBP_FortBuyBrief_ExpertSale_C}
    stockcount {value=1}
    cannotusereason {value=""}
    configdef {value=&ExpertSaleBriefCollection}
    configkey {value=ExpertSale}
   ]
   requirements [{
     cannotUseReasonParameter=""
     failureNodeBehaviour=FailedButVisible
     participantId {tagName=Conversation.Participant.Player}
     requirement {
      class=UFortControllerRequirement_TokenThreshold_C
      properties {
       itemTemplateId=Token:athena_s34_weaponexpertise_exotic_service_token
       requiredQuantityThresholdConfig {
        requirement=Equal
        quantity {
         value=1
         curve {curveTable=&ExpertSaleRequirementsData, rowName=ExpertSale.Requirement.Tokens}
        }
       }
      }
     }
    }]
   serviceCost {curve {rowName=ExpertSale.Cost.Exotic.One}}
  }
  sub_nodes [ConversationChoiceNode_15 {class=ConversationChoiceNode, properties {defaultChoiceDisplayText=BUY}}]
 }
 Back_C_1 {
  id=#17
  class=Back_C
  sub_nodes [ConversationChoiceNode_7 {class=ConversationChoiceNode, properties {defaultChoiceDisplayText=BACK}}]
 }
 Speech_C_0 {
  id=#2
  class=Speech_C
  outputs [#19, #20]
  properties {bUseInlineConfig=false, softTaskIcon=&T_UI_NPC_WeaponExpertise}
  sub_nodes [
   FortConversationChoiceWithPreviewNode_2 {class=FortConversationChoiceWithPreviewNode, properties {defaultChoiceDisplayText=EXPERT SALE}}
   FortConversationRequirement_HasService_1 {class=FortConversationRequirement_HasService, properties {serviceTag {tagName=Tandem.Service.ExpertSale.Healing}}}
  ]
 }
 ExpertSale_C_0 {
  id=#20
  class=ExpertSale_C
  outputs [#10]
  properties {
   bShowPropertyDetails=false
   loot Tier Group Name=LTG_Market_AgileChug
   requirements [{
     cannotUseReasonParameter=notExpertEnough
     failureNodeBehaviour=FailedAndHidden
     participantId {tagName=Conversation.Participant.Player}
     requirement {
      class=UFortControllerRequirement_TokenThreshold_C
      properties {
       bSkip=true
       itemTemplateId=Token:athena_s34_weaponexpertise_exotic_service_token
       requiredQuantityThresholdConfig {
        requirement=LessThanOrEqual
        quantity {
         value=1
         curve {curveTable=&ExpertSaleRequirementsData, rowName=ExpertSale.Requirement.Tokens}
        }
       }
      }
     }
    }]
   serviceCost {curve {rowName=ExpertSale.Cost.Healing.One}}
  }
  sub_nodes [ConversationChoiceNode_1 {class=ConversationChoiceNode, properties {defaultChoiceDisplayText=BUY}}]
 }
 Back_C_0 {
  id=#19
  class=Back_C
  sub_nodes [ConversationChoiceNode_4 {class=ConversationChoiceNode, properties {defaultChoiceDisplayText=BACK}}]
 }
]
