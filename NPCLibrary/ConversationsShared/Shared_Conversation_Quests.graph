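# Shared NPC conversation database that serves the Quests entry point, offering up to five slotted quests from NPC dialogue.
# Each slot's speech choice is gated by its quest service tag, no active quests, completed prerequisites and no completion this match.
# ACCEPT grants that slot's quest if shared quest capacity allows, and BACK returns, each playing a client voice-over.
# The entry also links out to the Season 17 Week 8 Impostor conversation.
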ConversationDatabase /NPCLibrary/ConversationsShared/Shared_Conversation_Quests
entry_points [{tag=Conversation.Entry.NPCLibrary.Quests, destinations [#1]}]
nodes [
 ConversationEntryPointNode_7 {
  id=#1
  class=ConversationEntryPointNode
  outputs [
   #2
   #3
   #4
   #5
   #6
   #7
  ]
  properties {entryTag {tagName=Conversation.Entry.NPCLibrary.Quests}}
 }
 Speech_C_16 {
  id=#7
  class=Speech_C
  outputs [#8, #9]
  properties {softTaskIcon=/Game/Athena/HUD/Interaction/T_UI_NPC_quest.T_UI_NPC_Quest, generalConfig {defaultMessage=I need a hand. Or two.}}
  sub_nodes [
   FortConversationRequirement_HasService_0 {class=FortConversationRequirement_HasService, properties {serviceTag {tagName=Tandem.Service.Quest1}}}
   FortConversationRequirement_AllSlottedQuestPrerequisitesCompleted_0 {class=FortConversationRequirement_AllSlottedQuestPrerequisitesCompleted}
   FortConversationChoiceWithPreviewNode_4 {class=FortConversationChoiceWithPreviewNode, properties {bShowPreviewOnlyIfChildTaskNodeValid=false, defaultChoiceDisplayText=Quest}}
   FortConversationRequirement_SlottedQuestNotCompletedThisMatch_0 {class=FortConversationRequirement_SlottedQuestNotCompletedThisMatch}
   FortConversationRequirement_HasNoActiveQuests_5 {class=FortConversationRequirement_HasNoActiveQuests}
  ]
 }
 Back_C_4 {
  id=#9
  class=Back_C
  sub_nodes [
   ConversationChoiceNode_2 {class=ConversationChoiceNode, properties {defaultChoiceDisplayText="BACK!"}}
   SideEffect_Client_PlayVOEvent_C_3 {class=SideEffect_Client_PlayVOEvent_C, properties {voiceOverTag {tagName=Pawn.Athena.NPC.Sound.Voice.Quest.Disappointed}}}
  ]
 }
 GrantSlottedQuest_C_0 {
  id=#8
  class=GrantSlottedQuest_C
  outputs []
  sub_nodes [
   ConversationChoiceNode_0 {class=ConversationChoiceNode, properties {defaultChoiceDisplayText="ACCEPT!"}}
   SideEffect_Client_PlayVOEvent_C_0 {class=SideEffect_Client_PlayVOEvent_C, properties {voiceOverTag {tagName=Pawn.Athena.NPC.Sound.Voice.Quest.Affirmative}}}
   FortConversationRequirement_ControllerRequirement_0 {
    class=FortConversationRequirement_ControllerRequirement
    properties {nodeName=Require: Has Shared Quest Capacity, requirement {class=FortControllerRequirement_HasSharedQuestCapacity}}
   }
  ]
 }
 Speech_C_2 {
  id=#6
  class=Speech_C
  outputs [#10, #11]
  properties {
   softTaskIcon=&T_UI_NPC_Quest
   generalConfig {defaultMessage=Can you help me with this one little thing?}
  }
  sub_nodes [
   FortConversationRequirement_HasService_4 {class=FortConversationRequirement_HasService, properties {serviceTag {tagName=Tandem.Service.Quest5}}}
   FortConversationRequirement_HasNoActiveQuests_4 {class=FortConversationRequirement_HasNoActiveQuests}
   FortConversationRequirement_AllSlottedQuestPrerequisitesCompleted_4 {class=FortConversationRequirement_AllSlottedQuestPrerequisitesCompleted, properties {questSlot=5}}
   FortConversationChoiceWithPreviewNode_1 {class=FortConversationChoiceWithPreviewNode, properties {bShowPreviewOnlyIfChildTaskNodeValid=false, defaultChoiceDisplayText=Quest}}
   FortConversationRequirement_SlottedQuestNotCompletedThisMatch_4 {class=FortConversationRequirement_SlottedQuestNotCompletedThisMatch, properties {questSlot=5}}
  ]
 }
 Back_C_2 {
  id=#11
  class=Back_C
  sub_nodes [
   ConversationChoiceNode_5 {class=ConversationChoiceNode, properties {defaultChoiceDisplayText="BACK!"}}
   SideEffect_Client_PlayVOEvent_C_8 {class=SideEffect_Client_PlayVOEvent_C, properties {voiceOverTag {tagName=Pawn.Athena.NPC.Sound.Voice.Quest.Disappointed}}}
  ]
 }
 GrantSlottedQuest_C_4 {
  id=#10
  class=GrantSlottedQuest_C
  outputs []
  properties {questSlot=5}
  sub_nodes [
   ConversationChoiceNode_9 {class=ConversationChoiceNode, properties {defaultChoiceDisplayText="ACCEPT!"}}
   SideEffect_Client_PlayVOEvent_C_9 {class=SideEffect_Client_PlayVOEvent_C, properties {voiceOverTag {tagName=Pawn.Athena.NPC.Sound.Voice.Quest.Affirmative}}}
   FortConversationRequirement_ControllerRequirement_5 {
    class=FortConversationRequirement_ControllerRequirement
    properties {nodeName=Require: Has Shared Quest Capacity, requirement {class=FortControllerRequirement_HasSharedQuestCapacity}}
   }
  ]
 }
 Speech_C_1 {
  id=#5
  class=Speech_C
  outputs [#12, #13]
  properties {
   softTaskIcon=&T_UI_NPC_Quest
   generalConfig {defaultMessage=Can you help me with this one little thing?}
  }
  sub_nodes [
   FortConversationRequirement_HasService_2 {class=FortConversationRequirement_HasService, properties {serviceTag {tagName=Tandem.Service.Quest4}}}
   FortConversationRequirement_HasNoActiveQuests_2 {class=FortConversationRequirement_HasNoActiveQuests}
   FortConversationRequirement_AllSlottedQuestPrerequisitesCompleted_2 {class=FortConversationRequirement_AllSlottedQuestPrerequisitesCompleted, properties {questSlot=4}}
   FortConversationChoiceWithPreviewNode_5 {class=FortConversationChoiceWithPreviewNode, properties {bShowPreviewOnlyIfChildTaskNodeValid=false, defaultChoiceDisplayText=Quest}}
   FortConversationRequirement_SlottedQuestNotCompletedThisMatch_2 {class=FortConversationRequirement_SlottedQuestNotCompletedThisMatch, properties {questSlot=4}}
  ]
 }
 Back_C_0 {
  id=#13
  class=Back_C
  sub_nodes [
   ConversationChoiceNode_8 {class=ConversationChoiceNode, properties {defaultChoiceDisplayText="BACK!"}}
   SideEffect_Client_PlayVOEvent_C_5 {class=SideEffect_Client_PlayVOEvent_C, properties {voiceOverTag {tagName=Pawn.Athena.NPC.Sound.Voice.Quest.Disappointed}}}
  ]
 }
 GrantSlottedQuest_C_3 {
  id=#12
  class=GrantSlottedQuest_C
  outputs []
  properties {questSlot=4}
  sub_nodes [
   ConversationChoiceNode_1 {class=ConversationChoiceNode, properties {defaultChoiceDisplayText="ACCEPT!"}}
   SideEffect_Client_PlayVOEvent_C_2 {class=SideEffect_Client_PlayVOEvent_C, properties {voiceOverTag {tagName=Pawn.Athena.NPC.Sound.Voice.Quest.Affirmative}}}
   FortConversationRequirement_ControllerRequirement_4 {
    class=FortConversationRequirement_ControllerRequirement
    properties {nodeName=Require: Has Shared Quest Capacity, requirement {class=FortControllerRequirement_HasSharedQuestCapacity}}
   }
  ]
 }
 Speech_C_0 {
  id=#4
  class=Speech_C
  outputs [#14, #15]
  properties {
   softTaskIcon=&T_UI_NPC_Quest
   generalConfig {defaultMessage=Can you help me with this one little thing?}
  }
  sub_nodes [
   FortConversationRequirement_HasService_3 {class=FortConversationRequirement_HasService, properties {serviceTag {tagName=Tandem.Service.Quest3}}}
   FortConversationRequirement_HasNoActiveQuests_3 {class=FortConversationRequirement_HasNoActiveQuests}
   FortConversationRequirement_AllSlottedQuestPrerequisitesCompleted_3 {class=FortConversationRequirement_AllSlottedQuestPrerequisitesCompleted, properties {questSlot=3}}
   FortConversationChoiceWithPreviewNode_0 {class=FortConversationChoiceWithPreviewNode, properties {bShowPreviewOnlyIfChildTaskNodeValid=false, defaultChoiceDisplayText=Quest}}
   FortConversationRequirement_SlottedQuestNotCompletedThisMatch_3 {class=FortConversationRequirement_SlottedQuestNotCompletedThisMatch, properties {questSlot=3}}
  ]
 }
 Back_C_5 {
  id=#15
  class=Back_C
  sub_nodes [
   ConversationChoiceNode_3 {class=ConversationChoiceNode, properties {defaultChoiceDisplayText="BACK!"}}
   SideEffect_Client_PlayVOEvent_C_6 {class=SideEffect_Client_PlayVOEvent_C, properties {voiceOverTag {tagName=Pawn.Athena.NPC.Sound.Voice.Quest.Disappointed}}}
  ]
 }
 GrantSlottedQuest_C_2 {
  id=#14
  class=GrantSlottedQuest_C
  outputs []
  properties {questSlot=3}
  sub_nodes [
   ConversationChoiceNode_4 {class=ConversationChoiceNode, properties {defaultChoiceDisplayText="ACCEPT!"}}
   SideEffect_Client_PlayVOEvent_C_7 {class=SideEffect_Client_PlayVOEvent_C, properties {voiceOverTag {tagName=Pawn.Athena.NPC.Sound.Voice.Quest.Affirmative}}}
   FortConversationRequirement_ControllerRequirement_3 {
    class=FortConversationRequirement_ControllerRequirement
    properties {nodeName=Require: Has Shared Quest Capacity, requirement {class=FortControllerRequirement_HasSharedQuestCapacity}}
   }
  ]
 }
 Speech_C_3 {
  id=#3
  class=Speech_C
  outputs [#16, #17]
  properties {
   softTaskIcon=&T_UI_NPC_Quest
   generalConfig {defaultMessage=Can you help me with this one little thing?}
  }
  sub_nodes [
   FortConversationRequirement_HasService_1 {class=FortConversationRequirement_HasService, properties {serviceTag {tagName=Tandem.Service.Quest2}}}
   FortConversationRequirement_HasNoActiveQuests_1 {class=FortConversationRequirement_HasNoActiveQuests}
   FortConversationRequirement_AllSlottedQuestPrerequisitesCompleted_1 {class=FortConversationRequirement_AllSlottedQuestPrerequisitesCompleted, properties {questSlot=2}}
   FortConversationChoiceWithPreviewNode_3 {class=FortConversationChoiceWithPreviewNode, properties {bShowPreviewOnlyIfChildTaskNodeValid=false, defaultChoiceDisplayText=Quest}}
   FortConversationRequirement_SlottedQuestNotCompletedThisMatch_1 {class=FortConversationRequirement_SlottedQuestNotCompletedThisMatch, properties {questSlot=2}}
  ]
 }
 Back_C_1 {
  id=#17
  class=Back_C
  sub_nodes [
   ConversationChoiceNode_7 {class=ConversationChoiceNode, properties {defaultChoiceDisplayText="BACK!"}}
   SideEffect_Client_PlayVOEvent_C_4 {class=SideEffect_Client_PlayVOEvent_C, properties {voiceOverTag {tagName=Pawn.Athena.NPC.Sound.Voice.Quest.Disappointed}}}
  ]
 }
 GrantSlottedQuest_C_1 {
  id=#16
  class=GrantSlottedQuest_C
  outputs []
  properties {questSlot=2}
  sub_nodes [
   ConversationChoiceNode_6 {class=ConversationChoiceNode, properties {defaultChoiceDisplayText="ACCEPT!"}}
   SideEffect_Client_PlayVOEvent_C_1 {class=SideEffect_Client_PlayVOEvent_C, properties {voiceOverTag {tagName=Pawn.Athena.NPC.Sound.Voice.Quest.Affirmative}}}
   FortConversationRequirement_ControllerRequirement_2 {
    class=FortConversationRequirement_ControllerRequirement
    properties {nodeName=Require: Has Shared Quest Capacity, requirement {class=FortControllerRequirement_HasSharedQuestCapacity}}
   }
  ]
 }
 ConversationLinkNode_0 {
  id=#2
  class=ConversationLinkNode
  properties {remoteEntryTag {tagName=Conversation.Entry.Season17.Week8.Impostor}}
 }
]
