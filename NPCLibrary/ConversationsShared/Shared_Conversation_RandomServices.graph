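# Shared NPC conversation database offering the RandomServices menu: ammo restock, rift-to-go, patch up and hire.
# Each Speech choice requires its service; random services also need a matching NPC-owned tag, while hire excludes NPCs tagged NoHire.
# Each Speech leads to a confirming service task or Back.
# Lines resolve by speaker entry tag: the three random services repeat one table, while hire has its own, including some equal-weight variants.
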ConversationDatabase /NPCLibrary/ConversationsShared/Shared_Conversation_RandomServices
entry_points [{tag=Conversation.Entry.RandomServices, destinations [#1]}]
nodes [
 ConversationEntryPointNode_0 {
  id=#1
  class=ConversationEntryPointNode
  outputs [#2, #3, #4, #5]
  properties {entryTag {tagName=Conversation.Entry.RandomServices}}
 }
 Speech_C_3 {
  id=#2
  class=Speech_C
  outputs [#6, #7]
  properties {
   bUsePerNPC_ContextualMessageLookup=true
   softTaskIcon=/Game/Athena/HUD/Art/BacchusHudIcons/T_UI_AmmoReload
   generalConfig {defaultMessage=Here's a refill.}
   speakerEntryTagToConfig {
    "(TagName=\"Conversation.Entry.ArcanaAgate\")" {defaultMessage=I always carry extras.}
    "(TagName=\"Conversation.Entry.AshenMagnus\")" {defaultMessage="Get up into the sky!"}
    "(TagName=\"Conversation.Entry.BaskIsle\")" {defaultMessage=You look like you need this.}
    "(TagName=\"Conversation.Entry.BlimpPoint\")" {defaultMessage=Good luck out there.}
    "(TagName=\"Conversation.Entry.Brat\")" {defaultMessage=This deal is bun-believable.}
    "(TagName=\"Conversation.Entry.ClumsyChew\")" {defaultMessage=Don't bug out.}
    "(TagName=\"Conversation.Entry.CoolSlice\")" {defaultMessage="Looks like you found yourself in a dilly of a pickle. "}
    "(TagName=\"Conversation.Entry.CopperMoggy\")" {defaultMessage=This should help.}
    "(TagName=\"Conversation.Entry.CrawlyTech\")" {defaultMessage=I hypothesize you'll be fine.}
    "(TagName=\"Conversation.Entry.CubicVice\")" {defaultMessage=See the sights of the Island.}
    "(TagName=\"Conversation.Entry.CyberMitt\")" {defaultMessage=The best view of the Island.}
    "(TagName=\"Conversation.Entry.DashSurge\")" {defaultMessage="Perfect for 'raw to go'!"}
    "(TagName=\"Conversation.Entry.DiaryString\")" {defaultMessage=*Smiles uncertainly*}
    "(TagName=\"Conversation.Entry.DiaryString_WidePuck\")" {defaultMessage="Sold!"}
    "(TagName=\"Conversation.Entry.DustyBun\")" {defaultMessage="Fly like the mighty moth!"}
    "(TagName=\"Conversation.Entry.FacetBlom\")" {defaultMessage=For the battle you face.}
    "(TagName=\"Conversation.Entry.FrostGalore\")" {defaultMessage="Careful, it's stuck on auto-fire!"}
    "(TagName=\"Conversation.Entry.FruitFire\")" {defaultMessage="Now it's Marge's time to shine!"}
    "(TagName=\"Conversation.Entry.FunkyFeather\")" {defaultMessage=You owe me a favor.}
    "(TagName=\"Conversation.Entry.GrimeHold\")" {defaultMessage=(Blinky Fishstick blinks eerily.)}
    "(TagName=\"Conversation.Entry.HangSpec\")" {defaultMessage=Make a grand exit.}
    "(TagName=\"Conversation.Entry.IcicleLarva\")" {defaultMessage=A great way to soar across the sky and help those in need.}
    "(TagName=\"Conversation.Entry.IronLilac\")" {defaultMessage=Ride the rift.}
    "(TagName=\"Conversation.Entry.LycheeNickel\")" {defaultMessage="For rou!"}
    "(TagName=\"Conversation.Entry.MistMylar\")" {defaultMessage=Travel in style.}
    "(TagName=\"Conversation.Entry.MistRaven\")" {defaultMessage="Ride the rift, like a friend of mine."}
    "(TagName=\"Conversation.Entry.Mizuki\")" {defaultMessage=Courtesy of the River Guard.}
    "(TagName=\"Conversation.Entry.OboeThorn\")" {defaultMessage=You'll need this.}
    "(TagName=\"Conversation.Entry.PixieParts\")" {defaultMessage=(The Guardian is happy to help!)}
    "(TagName=\"Conversation.Entry.PolishedJade\")" {defaultMessage=Ready to rift?}
    "(TagName=\"Conversation.Entry.PrivateJet\")" {defaultMessage="But of course! Here you are."}
    "(TagName=\"Conversation.Entry.QueenTruth\")" {defaultMessage="If you don’t watch the violence, you’ll never get desensitized to it."}
    "(TagName=\"Conversation.Entry.RoseDepth\")" {defaultMessage=Go.}
    "(TagName=\"Conversation.Entry.SatireCane\")" {defaultMessage="It's riftin' time!"}
    "(TagName=\"Conversation.Entry.SilentNovel\")" {defaultMessage="Now you're ready for a night on the town!"}
    "(TagName=\"Conversation.Entry.SkiffGrind\")" {defaultMessage="Courtesy of MEKA Squad!"}
    "(TagName=\"Conversation.Entry.SnapFreeze\")" {defaultMessage="Be sure to wave!"}
    "(TagName=\"Conversation.Entry.SwoopClasp\")" {defaultMessage=Make an emergency exit.}
    "(TagName=\"Conversation.Entry.SynthContact\")" {defaultMessage=I am at your service.}
    "(TagName=\"Conversation.Entry.TalonPane\")" {defaultMessage=Need an exit strategy?}
    "(TagName=\"Conversation.Entry.TennisLeash\")" {defaultMessage="Happy Winterfest!"}
    "(TagName=\"Conversation.Entry.TeriyakiFish\")" {defaultMessage=(Fishstick offers his support.)}
    "(TagName=\"Conversation.Entry.TroutWrist\")" {defaultMessage=(Peelverine nods his head.)}
    "(TagName=\"Conversation.Entry.UndergroundRebel\")" {defaultMessage=Get where you need to go.}
    "(TagName=\"Conversation.Entry.UniqueBeetle\")" {defaultMessage="Let's go have some fun, huh?"}
    "(TagName=\"Conversation.Entry.Unpeely\")" {defaultMessage=(Peely gives you a thumbs up.)}
    "(TagName=\"Conversation.Entry.VileString_Crude\")" {defaultMessage=I'll have your back.}
    "(TagName=\"Conversation.Entry.VoidSpray\")" {defaultMessage=I hope this is useful.}
    "(TagName=\"Conversation.Entry.WiryPerk\")" {defaultMessage=Ready to blast off?}
    "(TagName=\"Conversation.Entry.Yeti\")" {defaultMessage=I can help.}
   }
  }
  sub_nodes [
   FortConversationRequirement_HasService_3 {class=FortConversationRequirement_HasService, properties {serviceTag {tagName=Tandem.Service.Random.Ammo}}}
   FortConversationRequirement_ControllerRequirement_0 {
    class=FortConversationRequirement_ControllerRequirement
    properties {
     nodeName=Require: Has GameplayTag
     participantId {tagName=Conversation.Participant.NPC}
     requirement {class=FortControllerRequirement_HasOwnedTag, properties {ownedTag {tagName=Tandem.Service.Random.Ammo}}}
    }
   }
   FortConversationChoiceWithPreviewNode_1 {
    class=FortConversationChoiceWithPreviewNode
    properties {defaultChoiceDisplayText=Ammo Restock, choiceTags {gameplayTags [{tagName=Conversation.Choice.CustomSortPriority.Low}]}}
   }
  ]
 }
 Back_C_5 {
  id=#7
  class=Back_C
  sub_nodes [ConversationChoiceNode_13 {class=ConversationChoiceNode, properties {defaultChoiceDisplayText=BACK}}]
 }
 AmmoGrant_C_0 {
  id=#6
  class=AmmoGrant_C
  outputs []
  sub_nodes [ConversationChoiceNode_7 {class=ConversationChoiceNode, properties {defaultChoiceDisplayText=CONFIRM}}]
 }
 Speech_C_23 {
  id=#4
  class=Speech_C
  outputs [#8, #9]
  properties {
   bUsePerNPC_ContextualMessageLookup=true
   softTaskIcon=/Game/Athena/HUD/Interaction/T_UI_NPC_medic
   generalConfig {defaultMessage=Heal yourself up.}
   speakerEntryTagToConfig {
    "(TagName=\"Conversation.Entry.ArcanaAgate\")" {defaultMessage=I always carry extras.}
    "(TagName=\"Conversation.Entry.AshenMagnus\")" {defaultMessage="Get up into the sky!"}
    "(TagName=\"Conversation.Entry.BaskIsle\")" {defaultMessage=You look like you need this.}
    "(TagName=\"Conversation.Entry.BlimpPoint\")" {defaultMessage=Good luck out there.}
    "(TagName=\"Conversation.Entry.Brat\")" {defaultMessage=This deal is bun-believable.}
    "(TagName=\"Conversation.Entry.ClumsyChew\")" {defaultMessage=Don't bug out.}
    "(TagName=\"Conversation.Entry.CoolSlice\")" {defaultMessage="Looks like you found yourself in a dilly of a pickle. "}
    "(TagName=\"Conversation.Entry.CopperMoggy\")" {defaultMessage=This should help.}
    "(TagName=\"Conversation.Entry.CrawlyTech\")" {defaultMessage=I hypothesize you'll be fine.}
    "(TagName=\"Conversation.Entry.CubicVice\")" {defaultMessage=See the sights of the Island.}
    "(TagName=\"Conversation.Entry.CyberMitt\")" {defaultMessage=The best view of the Island.}
    "(TagName=\"Conversation.Entry.DashSurge\")" {defaultMessage="Perfect for 'raw to go'!"}
    "(TagName=\"Conversation.Entry.DiaryString\")" {defaultMessage=*Smiles uncertainly*}
    "(TagName=\"Conversation.Entry.DiaryString_WidePuck\")" {defaultMessage="Sold!"}
    "(TagName=\"Conversation.Entry.DustyBun\")" {defaultMessage="Fly like the mighty moth!"}
    "(TagName=\"Conversation.Entry.FacetBlom\")" {defaultMessage=For the battle you face.}
    "(TagName=\"Conversation.Entry.FrostGalore\")" {defaultMessage="Careful, it's stuck on auto-fire!"}
    "(TagName=\"Conversation.Entry.FruitFire\")" {defaultMessage="Now it's Marge's time to shine!"}
    "(TagName=\"Conversation.Entry.FunkyFeather\")" {defaultMessage=You owe me a favor.}
    "(TagName=\"Conversation.Entry.GrimeHold\")" {defaultMessage=(Blinky Fishstick blinks eerily.)}
    "(TagName=\"Conversation.Entry.HangSpec\")" {defaultMessage=Make a grand exit.}
    "(TagName=\"Conversation.Entry.IcicleLarva\")" {defaultMessage=A great way to soar across the sky and help those in need.}
    "(TagName=\"Conversation.Entry.IronLilac\")" {defaultMessage=Ride the rift.}
    "(TagName=\"Conversation.Entry.LycheeNickel\")" {defaultMessage="For rou!"}
    "(TagName=\"Conversation.Entry.MistMylar\")" {defaultMessage=Travel in style.}
    "(TagName=\"Conversation.Entry.MistRaven\")" {defaultMessage="Ride the rift, like a friend of mine."}
    "(TagName=\"Conversation.Entry.Mizuki\")" {defaultMessage=Courtesy of the River Guard.}
    "(TagName=\"Conversation.Entry.OboeThorn\")" {defaultMessage=You'll need this.}
    "(TagName=\"Conversation.Entry.PixieParts\")" {defaultMessage=(The Guardian is happy to help!)}
    "(TagName=\"Conversation.Entry.PolishedJade\")" {defaultMessage=Ready to rift?}
    "(TagName=\"Conversation.Entry.PrivateJet\")" {defaultMessage="But of course! Here you are."}
    "(TagName=\"Conversation.Entry.QueenTruth\")" {defaultMessage="If you don’t watch the violence, you’ll never get desensitized to it."}
    "(TagName=\"Conversation.Entry.RoseDepth\")" {defaultMessage=Go.}
    "(TagName=\"Conversation.Entry.SatireCane\")" {defaultMessage="It's riftin' time!"}
    "(TagName=\"Conversation.Entry.SilentNovel\")" {defaultMessage="Now you're ready for a night on the town!"}
    "(TagName=\"Conversation.Entry.SkiffGrind\")" {defaultMessage="Courtesy of MEKA Squad!"}
    "(TagName=\"Conversation.Entry.SnapFreeze\")" {defaultMessage="Be sure to wave!"}
    "(TagName=\"Conversation.Entry.SwoopClasp\")" {defaultMessage=Make an emergency exit.}
    "(TagName=\"Conversation.Entry.SynthContact\")" {defaultMessage=I am at your service.}
    "(TagName=\"Conversation.Entry.TalonPane\")" {defaultMessage=Need an exit strategy?}
    "(TagName=\"Conversation.Entry.TennisLeash\")" {defaultMessage="Happy Winterfest!"}
    "(TagName=\"Conversation.Entry.TeriyakiFish\")" {defaultMessage=(Fishstick offers his support.)}
    "(TagName=\"Conversation.Entry.TroutWrist\")" {defaultMessage=(Peelverine nods his head.)}
    "(TagName=\"Conversation.Entry.UndergroundRebel\")" {defaultMessage=Get where you need to go.}
    "(TagName=\"Conversation.Entry.UniqueBeetle\")" {defaultMessage="Let's go have some fun, huh?"}
    "(TagName=\"Conversation.Entry.Unpeely\")" {defaultMessage=(Peely gives you a thumbs up.)}
    "(TagName=\"Conversation.Entry.VileString_Crude\")" {defaultMessage=I'll have your back.}
    "(TagName=\"Conversation.Entry.VoidSpray\")" {defaultMessage=I hope this is useful.}
    "(TagName=\"Conversation.Entry.WiryPerk\")" {defaultMessage=Ready to blast off?}
    "(TagName=\"Conversation.Entry.Yeti\")" {defaultMessage=I can help.}
   }
  }
  sub_nodes [
   FortConversationChoiceWithPreviewNode_23 {
    class=FortConversationChoiceWithPreviewNode
    properties {defaultChoiceDisplayText="PATCH UP!", choiceTags {gameplayTags [{tagName=Conversation.Choice.CustomSortPriority.Low}]}}
   }
   FortConversationRequirement_HasService_2 {class=FortConversationRequirement_HasService, properties {serviceTag {tagName=Tandem.Service.Random.Patchup}}}
   FortConversationRequirement_ControllerRequirement_1 {
    class=FortConversationRequirement_ControllerRequirement
    properties {
     nodeName=Require: Has GameplayTag
     participantId {tagName=Conversation.Participant.NPC}
     requirement {class=FortControllerRequirement_HasOwnedTag, properties {ownedTag {tagName=Tandem.Service.Random.Patchup}}}
    }
   }
  ]
 }
 EHP_PatchUp_C_1 {
  id=#8
  class=EHP_PatchUp_C
  outputs []
  properties {softTaskIcon=&T_UI_NPC_medic}
  sub_nodes [ConversationChoiceNode_10 {class=ConversationChoiceNode, properties {defaultChoiceDisplayText=ACCEPT}}]
 }
 Back_C_1 {
  id=#9
  class=Back_C
  sub_nodes [ConversationChoiceNode_12 {class=ConversationChoiceNode, properties {defaultChoiceDisplayText=BACK}}]
 }
 Speech_C_2 {
  id=#3
  class=Speech_C
  outputs [#10, #11]
  properties {
   bUsePerNPC_ContextualMessageLookup=true
   softTaskIcon=/Game/Athena/HUD/Interaction/T_UI_NPC_Rift
   generalConfig {defaultMessage=Rift carefully.}
   speakerEntryTagToConfig {
    "(TagName=\"Conversation.Entry.ArcanaAgate\")" {defaultMessage=I always carry extras.}
    "(TagName=\"Conversation.Entry.AshenMagnus\")" {defaultMessage="Get up into the sky!"}
    "(TagName=\"Conversation.Entry.BaskIsle\")" {defaultMessage=You look like you need this.}
    "(TagName=\"Conversation.Entry.BlimpPoint\")" {defaultMessage=Good luck out there.}
    "(TagName=\"Conversation.Entry.Brat\")" {defaultMessage=This deal is bun-believable.}
    "(TagName=\"Conversation.Entry.ClumsyChew\")" {defaultMessage=Don't bug out.}
    "(TagName=\"Conversation.Entry.CoolSlice\")" {defaultMessage="Looks like you found yourself in a dilly of a pickle. "}
    "(TagName=\"Conversation.Entry.CopperMoggy\")" {defaultMessage=This should help.}
    "(TagName=\"Conversation.Entry.CrawlyTech\")" {defaultMessage=I hypothesize you'll be fine.}
    "(TagName=\"Conversation.Entry.CubicVice\")" {defaultMessage=See the sights of the Island.}
    "(TagName=\"Conversation.Entry.CyberMitt\")" {defaultMessage=The best view of the Island.}
    "(TagName=\"Conversation.Entry.DashSurge\")" {defaultMessage="Perfect for 'raw to go'!"}
    "(TagName=\"Conversation.Entry.DiaryString\")" {defaultMessage=*Smiles uncertainly*}
    "(TagName=\"Conversation.Entry.DiaryString_WidePuck\")" {defaultMessage="Sold!"}
    "(TagName=\"Conversation.Entry.DustyBun\")" {defaultMessage="Fly like the mighty moth!"}
    "(TagName=\"Conversation.Entry.FacetBlom\")" {defaultMessage=For the battle you face.}
    "(TagName=\"Conversation.Entry.FrostGalore\")" {defaultMessage="Careful, it's stuck on auto-fire!"}
    "(TagName=\"Conversation.Entry.FruitFire\")" {defaultMessage="Now it's Marge's time to shine!"}
    "(TagName=\"Conversation.Entry.FunkyFeather\")" {defaultMessage=You owe me a favor.}
    "(TagName=\"Conversation.Entry.GrimeHold\")" {defaultMessage=(Blinky Fishstick blinks eerily.)}
    "(TagName=\"Conversation.Entry.HangSpec\")" {defaultMessage=Make a grand exit.}
    "(TagName=\"Conversation.Entry.IcicleLarva\")" {defaultMessage=A great way to soar across the sky and help those in need.}
    "(TagName=\"Conversation.Entry.IronLilac\")" {defaultMessage=Ride the rift.}
    "(TagName=\"Conversation.Entry.LycheeNickel\")" {defaultMessage="For rou!"}
    "(TagName=\"Conversation.Entry.MistMylar\")" {defaultMessage=Travel in style.}
    "(TagName=\"Conversation.Entry.MistRaven\")" {defaultMessage="Ride the rift, like a friend of mine."}
    "(TagName=\"Conversation.Entry.Mizuki\")" {defaultMessage=Courtesy of the River Guard.}
    "(TagName=\"Conversation.Entry.OboeThorn\")" {defaultMessage=You'll need this.}
    "(TagName=\"Conversation.Entry.PixieParts\")" {defaultMessage=(The Guardian is happy to help!)}
    "(TagName=\"Conversation.Entry.PolishedJade\")" {defaultMessage=Ready to rift?}
    "(TagName=\"Conversation.Entry.PrivateJet\")" {defaultMessage="But of course! Here you are."}
    "(TagName=\"Conversation.Entry.QueenTruth\")" {defaultMessage="If you don’t watch the violence, you’ll never get desensitized to it."}
    "(TagName=\"Conversation.Entry.RoseDepth\")" {defaultMessage=Go.}
    "(TagName=\"Conversation.Entry.SatireCane\")" {defaultMessage="It's riftin' time!"}
    "(TagName=\"Conversation.Entry.SilentNovel\")" {defaultMessage="Now you're ready for a night on the town!"}
    "(TagName=\"Conversation.Entry.SkiffGrind\")" {defaultMessage="Courtesy of MEKA Squad!"}
    "(TagName=\"Conversation.Entry.SnapFreeze\")" {defaultMessage="Be sure to wave!"}
    "(TagName=\"Conversation.Entry.SwoopClasp\")" {defaultMessage=Make an emergency exit.}
    "(TagName=\"Conversation.Entry.SynthContact\")" {defaultMessage=I am at your service.}
    "(TagName=\"Conversation.Entry.TalonPane\")" {defaultMessage=Need an exit strategy?}
    "(TagName=\"Conversation.Entry.TennisLeash\")" {defaultMessage="Happy Winterfest!"}
    "(TagName=\"Conversation.Entry.TeriyakiFish\")" {defaultMessage=(Fishstick offers his support.)}
    "(TagName=\"Conversation.Entry.TroutWrist\")" {defaultMessage=(Peelverine nods his head.)}
    "(TagName=\"Conversation.Entry.UndergroundRebel\")" {defaultMessage=Get where you need to go.}
    "(TagName=\"Conversation.Entry.UniqueBeetle\")" {defaultMessage="Let's go have some fun, huh?"}
    "(TagName=\"Conversation.Entry.Unpeely\")" {defaultMessage=(Peely gives you a thumbs up.)}
    "(TagName=\"Conversation.Entry.VileString_Crude\")" {defaultMessage=I'll have your back.}
    "(TagName=\"Conversation.Entry.VoidSpray\")" {defaultMessage=I hope this is useful.}
    "(TagName=\"Conversation.Entry.WiryPerk\")" {defaultMessage=Ready to blast off?}
    "(TagName=\"Conversation.Entry.Yeti\")" {defaultMessage=I can help.}
   }
  }
  sub_nodes [
   FortConversationRequirement_HasService_6 {class=FortConversationRequirement_HasService, properties {serviceTag {tagName=Tandem.Service.Random.Rift}}}
   FortConversationChoiceWithPreviewNode_8 {
    class=FortConversationChoiceWithPreviewNode
    properties {defaultChoiceDisplayText=RIFT-TO-GO, choiceTags {gameplayTags [{tagName=Conversation.Choice.CustomSortPriority.Low}]}}
   }
   FortConversationRequirement_ControllerRequirement_6 {
    class=FortConversationRequirement_ControllerRequirement
    properties {
     nodeName=Require: Has GameplayTag
     participantId {tagName=Conversation.Participant.NPC}
     requirement {class=FortControllerRequirement_HasOwnedTag, properties {ownedTag {tagName=Tandem.Service.Random.Rift}}}
    }
   }
  ]
 }
 RiftToGo_C_0 {
  id=#10
  class=RiftToGo_C
  outputs []
  properties {softTaskIcon=&T_UI_NPC_Rift}
  sub_nodes [ConversationChoiceNode_2 {class=ConversationChoiceNode, properties {defaultChoiceDisplayText=CONFIRM}}]
 }
 Back_C_4 {
  id=#11
  class=Back_C
  sub_nodes [ConversationChoiceNode_0 {class=ConversationChoiceNode, properties {defaultChoiceDisplayText=BACK}}]
 }
 Speech_C_4 {
  id=#5
  class=Speech_C
  outputs [#12, #13]
  properties {
   bUsePerNPC_ContextualMessageLookup=true
   softTaskIcon=/Game/Athena/HUD/Interaction/T_UI_NPC_Hire
   generalConfig {defaultMessage=I'll have your back.}
   speakerEntryTagToConfig {
    "(TagName=\"Conversation.Entry.AshenMagnus\")" {defaultMessage=I'll watch your back out there.}
    "(TagName=\"Conversation.Entry.Banana\")" {defaultMessage=(Peely grins excitedly.)}
    "(TagName=\"Conversation.Entry.BentBaton\")" {defaultMessage=Wanna squad up?}
    "(TagName=\"Conversation.Entry.BrightIon\")" {defaultMessage=What're we waiting for?}
    "(TagName=\"Conversation.Entry.CubicVice\")" {defaultMessage=My sword is yours.}
    "(TagName=\"Conversation.Entry.DarkNinja_F\")" {defaultMessage=I see victory on the horizon...}
    "(TagName=\"Conversation.Entry.DurrburgerHeroF\")" {defaultMessage="You've got great taste. "}
    "(TagName=\"Conversation.Entry.DurrburgerWorker_F\")" {defaultMessage="Reporting for foody!"}
    "(TagName=\"Conversation.Entry.EctoCat\")" {defaultMessage=I'll be your shadow.}
    "(TagName=\"Conversation.Entry.EvokeFind\")" {defaultMessage="Send 'em back to the depths!"}
    "(TagName=\"Conversation.Entry.FoxWarrior\")" {defaultMessage=Fox Clan's specialty. I'm in.}
    "(TagName=\"Conversation.Entry.FumeFleeceClap\")" {defaultMessage=I know just what to do.}
    "(TagName=\"Conversation.Entry.FumeFleeceFade\")" {defaultMessage=The world needs us.}
    "(TagName=\"Conversation.Entry.FumeFleeceJig\")" {defaultMessage=It's time to go to work.}
    "(TagName=\"Conversation.Entry.FumeFleeceShack\")" {defaultMessage="Let's make 'em sick, sick, SICK!"}
    "(TagName=\"Conversation.Entry.FumeFleeceWag\")" {defaultMessage="Ready to rock!"}
    "(TagName=\"Conversation.Entry.FuzzyBearHalloween\")" {defaultMessage="Let's get spooky!"}
    "(TagName=\"Conversation.Entry.HalloweenTomatoF\")" {defaultMessage="Let's get lost in the sauce. "}
    "(TagName=\"Conversation.Entry.HenchmanTough\")" {defaultMessage=Need some muscle?}
    "(TagName=\"Conversation.Entry.LazyLizz\")" {defaultMessage=Beep boop... Deploying field lizard... confirm?}
    "(TagName=\"Conversation.Entry.LionSoldier\")" {defaultMessage=I'm ready if you are.}
    "(TagName=\"Conversation.Entry.LoanFloat\")" {defaultMessage=We're gonna be golden.}
    "(TagName=\"Conversation.Entry.LycheeNickel\")" {defaultMessage="Ret's ro!"}
    "(TagName=\"Conversation.Entry.MadCommander\")" {defaultMessage="Just for you, potential Chum."}
    "(TagName=\"Conversation.Entry.Meowscles\")" {defaultMessage=Let's flex on 'em.}
    "(TagName=\"Conversation.Entry.Mizuki\")" {defaultMessage=The river provides.}
    "(TagName=\"Conversation.Entry.NanaSplit\")" {defaultMessage=(Peely Bone is ready for adventure.)}
    "(TagName=\"Conversation.Entry.PinkBear\")" {defaultMessage=Wanna get going?}
    "(TagName=\"Conversation.Entry.PolarityWinn\")" {defaultMessage="Synthwave, ready for action!"}
    "(TagName=\"Conversation.Entry.PolishedJade\")" {defaultMessage=Let's have an adventure.}
    "(TagName=\"Conversation.Entry.PulseLancer.Smash.EvokeFind\")" {
     defaultMessage="Who's the baddest? Us, hellooo!"
     contextualMessages [
      {
       message="You won't regret this. Let's get hunting!"
       requirementMatchPolicy=RequireAny
       contextRequirements []
       metadataParameters []
       metadataTags {gameplayTags []}
       priority {registryType=None, value=0, curve {curveTable=null, rowName=None}}
       weight {registryType=None, value=1, curve {curveTable=null, rowName=None}}
      }
      {
       message="HUNTR/X is on your side!"
       requirementMatchPolicy=RequireAny
       contextRequirements []
       metadataParameters []
       metadataTags {gameplayTags []}
       priority {registryType=None, value=0, curve {curveTable=null, rowName=None}}
       weight {registryType=None, value=1, curve {curveTable=null, rowName=None}}
      }
      {
       message="First, we fight. Then, we couch. Deal?"
       requirementMatchPolicy=RequireAny
       contextRequirements []
       metadataParameters []
       metadataTags {gameplayTags []}
       priority {registryType=None, value=0, curve {curveTable=null, rowName=None}}
       weight {registryType=None, value=1, curve {curveTable=null, rowName=None}}
      }
      {
       message="Ooh yeah! Let's do this!!"
       requirementMatchPolicy=RequireAny
       contextRequirements []
       metadataParameters []
       metadataTags {gameplayTags []}
       priority {registryType=None, value=0, curve {curveTable=null, rowName=None}}
       weight {registryType=None, value=1, curve {curveTable=null, rowName=None}}
      }
      {
       message="Who's the baddest? Us, hellooo!"
       requirementMatchPolicy=RequireAny
       contextRequirements []
       metadataParameters []
       metadataTags {gameplayTags []}
       priority {registryType=None, value=0, curve {curveTable=null, rowName=None}}
       weight {registryType=None, value=1, curve {curveTable=null, rowName=None}}
      }
     ]
    }
    "(TagName=\"Conversation.Entry.PulseLancer.Smash.LoanFloat\")" {
     defaultMessage="We were born to do this. Let's go!"
     contextualMessages [
      {
       message="We were born to do this. Let's go!"
       requirementMatchPolicy=RequireAny
       contextRequirements []
       metadataParameters []
       metadataTags {gameplayTags []}
       priority {registryType=None, value=0, curve {curveTable=null, rowName=None}}
       weight {registryType=None, value=1, curve {curveTable=null, rowName=None}}
      }
      {
       message=Ready to shine?
       requirementMatchPolicy=RequireAny
       contextRequirements []
       metadataParameters []
       metadataTags {gameplayTags []}
       priority {registryType=None, value=0, curve {curveTable=null, rowName=None}}
       weight {registryType=None, value=1, curve {curveTable=null, rowName=None}}
      }
      {
       message="We're stronger as a team. "
       requirementMatchPolicy=RequireAny
       contextRequirements []
       metadataParameters []
       metadataTags {gameplayTags []}
       priority {registryType=None, value=0, curve {curveTable=null, rowName=None}}
       weight {registryType=None, value=1, curve {curveTable=null, rowName=None}}
      }
      {
       message=Your solo career is over. We're a demon-slaying group now.
       requirementMatchPolicy=RequireAny
       contextRequirements []
       metadataParameters []
       metadataTags {gameplayTags []}
       priority {registryType=None, value=0, curve {curveTable=null, rowName=None}}
       weight {registryType=None, value=1, curve {curveTable=null, rowName=None}}
      }
      {
       message="Consider yourself an honorary, temporary member of HUNTR/X."
       requirementMatchPolicy=RequireAny
       contextRequirements []
       metadataParameters []
       metadataTags {gameplayTags []}
       priority {registryType=None, value=0, curve {curveTable=null, rowName=None}}
       weight {registryType=None, value=1, curve {curveTable=null, rowName=None}}
      }
     ]
    }
    "(TagName=\"Conversation.Entry.PulseLancer.Smash.QuailWink\")" {
     defaultMessage=Anything for a fan. Especially demon hunting. Heh.
     contextualMessages [
      {
       message=Anything for a fan. Especially demon hunting. Heh.
       requirementMatchPolicy=RequireAny
       contextRequirements []
       metadataParameters []
       metadataTags {gameplayTags []}
       priority {registryType=None, value=0, curve {curveTable=null, rowName=None}}
       weight {registryType=None, value=1, curve {curveTable=null, rowName=None}}
      }
      {
       message=Stand back. I need space to work.
       requirementMatchPolicy=RequireAny
       contextRequirements []
       metadataParameters []
       metadataTags {gameplayTags []}
       priority {registryType=None, value=0, curve {curveTable=null, rowName=None}}
       weight {registryType=None, value=1, curve {curveTable=null, rowName=None}}
      }
      {
       message=They don't stand a chance.
       requirementMatchPolicy=RequireAny
       contextRequirements []
       metadataParameters []
       metadataTags {gameplayTags []}
       priority {registryType=None, value=0, curve {curveTable=null, rowName=None}}
       weight {registryType=None, value=1, curve {curveTable=null, rowName=None}}
      }
      {
       message=I was hoping you'd ask. Let's go.
       requirementMatchPolicy=RequireAny
       contextRequirements []
       metadataParameters []
       metadataTags {gameplayTags []}
       priority {registryType=None, value=0, curve {curveTable=null, rowName=None}}
       weight {registryType=None, value=1, curve {curveTable=null, rowName=None}}
      }
      {
       message=I got your back.
       requirementMatchPolicy=RequireAny
       contextRequirements []
       metadataParameters []
       metadataTags {gameplayTags []}
       priority {registryType=None, value=0, curve {curveTable=null, rowName=None}}
       weight {registryType=None, value=1, curve {curveTable=null, rowName=None}}
      }
     ]
    }
    "(TagName=\"Conversation.Entry.QuailWink\")" {defaultMessage=Ready for our napalm era.}
    "(TagName=\"Conversation.Entry.RankedTrooper\")" {defaultMessage=Ready to have an Elite at your side?}
    "(TagName=\"Conversation.Entry.Retro_M\")" {defaultMessage=Need a wingman?}
    "(TagName=\"Conversation.Entry.Robin\")" {defaultMessage="I'll help you out, but I'm not your sidekick."}
    "(TagName=\"Conversation.Entry.RoseDepth\")" {defaultMessage=Let's chase some chaos.}
    "(TagName=\"Conversation.Entry.SeleneCobra\")" {defaultMessage="Let's hunt, little mouse."}
    "(TagName=\"Conversation.Entry.SharpFang\")" {defaultMessage="I'll have your back out there!"}
    "(TagName=\"Conversation.Entry.SpacialTravel\")" {defaultMessage=You're lucky I'm available.}
    "(TagName=\"Conversation.Entry.SprocketPoppy\")" {defaultMessage=May our enemies feel the cold.}
    "(TagName=\"Conversation.Entry.StrongHibiscus\")" {defaultMessage="Let's set it OFF!"}
    "(TagName=\"Conversation.Entry.SwampFish\")" {defaultMessage=Bog... will follow...}
    "(TagName=\"Conversation.Entry.TacoF\")" {defaultMessage="I'm all about the extras. "}
    "(TagName=\"Conversation.Entry.TriageTrooper\")" {defaultMessage="Let's get going!"}
    "(TagName=\"Conversation.Entry.Trigger\")" {defaultMessage=You lead. I'll follow.}
    "(TagName=\"Conversation.Entry.Yeti\")" {defaultMessage=I can help.}
   }
  }
  sub_nodes [
   FortConversationRequirement_HasService_0 {class=FortConversationRequirement_HasService, properties {serviceTag {tagName=Tandem.Service.Hire}}}
   FortConversationChoiceWithPreviewNode_0 {
    class=FortConversationChoiceWithPreviewNode
    properties {defaultChoiceDisplayText=HIRE, choiceTags {gameplayTags [{tagName=Conversation.Choice.CustomSortPriority.Lowest}]}}
   }
   FortConversationRequirement_ControllerRequirement_3 {
    class=FortConversationRequirement_ControllerRequirement
    properties {
     nodeName=Require: Has GameplayTag
     participantId {tagName=Conversation.Participant.NPC}
     requirement {class=FortControllerRequirement_HasOwnedTag, properties {bInverseCondition=true, ownedTag {tagName=NPC.Status.NoHire}}}
    }
   }
  ]
 }
 HireNPC_C_0 {
  id=#12
  class=HireNPC_C
  outputs []
  sub_nodes [ConversationChoiceNode_4 {class=ConversationChoiceNode, properties {defaultChoiceDisplayText=CONFIRM}}]
 }
 Back_C_3 {
  id=#13
  class=Back_C
  sub_nodes [ConversationChoiceNode_1 {class=ConversationChoiceNode, properties {defaultChoiceDisplayText=BACK}}]
 }
]
